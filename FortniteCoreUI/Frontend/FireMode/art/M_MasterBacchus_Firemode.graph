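# Translucent UI master material for a frontend fire-mode selection panel, layering gradient background, item image, border, and optional bottom bar and icon.
# Rotated, offset UVs multiplied by sharpness values cut crisp angled edge masks, with resolution correction on the border.
# The image supports scaling, flipping, tinting, vertical fade and animated opacity over a blendable radial or linear gradient.
# A hover scalar drives border and bar changes; static switches enable bar and icon.
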Material /FortniteCoreUI/Frontend/FireMode/art/M_MasterBacchus_Firemode
graph {
 properties {
  bUsedWithStaticMesh=false
  blendMode=BLEND_Translucent
  materialDomain=MD_UI
  pixelDepthOffsetMode=PDOM_Legacy
  usedShadingModels=Default Lit
 }
 nodes [
  n1 = MaterialFunctionCall() {materialFunction=&GetUserInterfaceUV}
  n2 = MaterialFunctionCall(Float2=@n1.Pixel Size) {materialFunction=&BreakOutFloat2Components}
  Divide(A=@n2.R, B=@n2.R)
  LinearInterpolate(
   A=Multiply(A=ScalarParameter() {defaultValue=1, group=opacity, parameterName=OverallOpacity, sliderMax=1}, B=NamedRerouteUsage())
   B=Subtract(A=NamedRerouteUsage(), B=NamedRerouteUsage())
   Alpha=ScalarParameter() {parameterName=Background_Opacity}
  )
  n3 = MaterialFunctionCall() {materialFunction=&GetUserInterfaceUV}
  n4 = Rotator(
   Coordinate=@n3.9-Slice UV
   Time=Multiply(A=ScalarParameter() {defaultValue=-3, group=Border, parameterName=Border_Left_Rotation}) {constB=0.01}
  ) {speed=6.2832}
  n5 = LinearInterpolate(
   A=Multiply(A=ScalarParameter() {defaultValue=-14, group=Border, parameterName=Border_Leftt_Offset}) {constB=0.01}
   B=Multiply(A=ScalarParameter() {defaultValue=-14, group=Border, parameterName=Border_Leftt_OffsetHover}) {constB=0.01}
   Alpha=ScalarParameter() {group=Animation, parameterName=hover, sliderMax=1}
  )
  n6 = MaterialFunctionCall!(Native_U_PixelSize) {materialFunction=/FortUILibrary/Foundation/Art_ToolKit/Tools/M_UI_Resolution_Sharpness_Correction}
  n7 = Multiply(A=@n6.Result, B=ScalarParameter() {defaultValue=128, group=Border, parameterName=Sharpness_U})
  n8 = Multiply(A=ComponentMask(None=Add(A=@n4, B=@n5)) {r=true}, B=@n7) {constB=32}
  n9 = ScalarParameter() {defaultValue=-3, group=Border, parameterName=Border_Right_Rotation}
  n10 = OneMinus(None=Rotator(Coordinate=@n3.9-Slice UV, Time=Multiply(A=@n9) {constB=0.01}) {speed=6.2832})
  n11 = LinearInterpolate(
   A=Multiply(A=ScalarParameter() {defaultValue=-14, group=Border, parameterName=Border_Right_Offset}) {constB=0.01}
   B=Multiply(A=ScalarParameter() {defaultValue=-14, group=Border, parameterName=Border_Right_OffsetHover}) {constB=0.01}
   Alpha=ScalarParameter() {group=Animation, parameterName=hover, sliderMax=1}
  )
  n12 = Multiply(A=ComponentMask(None=Add(A=@n10, B=@n11)) {r=true}, B=@n7) {constB=32}
  n13 = Saturate(None=DotProduct(A=Saturate(None=@n8), B=Saturate(None=@n12)))
  n14 = Rotator(Coordinate=@n3.Normalized UV, Time=Multiply(A=ScalarParameter() {group=Border, parameterName=Border_Top_Rotation}) {constB=0.01}) {speed=6.2832}
  n15 = LinearInterpolate(
   A=Multiply(A=ScalarParameter() {defaultValue=-3, group=Border, parameterName=Border_Top_Offset}) {constB=0.01}
   B=Multiply(A=ScalarParameter() {defaultValue=-3, group=Border, parameterName=Border_Top_OffsetHover}) {constB=0.01}
   Alpha=ScalarParameter() {group=Animation, parameterName=hover, sliderMax=1}
  )
  n16 = Multiply(A=@n6.Result, B=ScalarParameter() {defaultValue=128, group=Border, parameterName=Sharpness_V})
  n17 = Multiply(A=ComponentMask(None=Add(A=@n14, B=@n15)) {g=true}, B=@n16) {constB=32}
  n18 = ScalarParameter() {group=Border, parameterName=Border_Bottom_Rotation}
  n19 = OneMinus(None=Rotator(Coordinate=@n3.Normalized UV, Time=Multiply(A=@n18) {constB=0.01}) {speed=6.2832})
  n20 = LinearInterpolate(
   A=Multiply(A=ScalarParameter() {defaultValue=-3, group=Border, parameterName=Border_Bottom_Offset}) {constB=0.01}
   B=Multiply(A=ScalarParameter() {defaultValue=-3, group=Border, parameterName=Border_Bottom_OffsetH0ver}) {constB=0.01}
   Alpha=ScalarParameter() {group=Animation, parameterName=hover, sliderMax=1}
  )
  n21 = Multiply(A=ComponentMask(None=Add(A=@n19, B=@n20)) {g=true}, B=@n16) {constB=32}
  n22 = Saturate(None=DotProduct(A=Saturate(None=@n17), B=Saturate(None=@n21)))
  NamedRerouteDeclaration(None=Saturate(None=DotProduct(A=@n13, B=@n22))) {name=Borders, nodeColor=<rgba 1 0 0.8082 1>}
  n23 = TextureSampleParameter2D(UVs=MaterialFunctionCall(
    UVs=MaterialFunctionCall() {materialFunction=&MF_UI_UVScaling}.ScaleY
    Texture Scale=ScalarParameter() {defaultValue=1, group=Image, parameterName=ImageMaskScale}
   ) {materialFunction=&ScaleUVsByCenter}.UVs) {
   group=Image
   parameterName=ItemImageMask
   samplerSource=SSM_Wrap_WorldGroupSettings
   samplerType=SAMPLERTYPE_Grayscale
   texture=&LowResBlurredNoise
  }
  n24 = Saturate(None=OneMinus(None=@n23.R))
  n25 = Subtract(A=OneMinus(None=@n23.R), B=NamedRerouteUsage())
  n26 = Saturate(None=NamedRerouteUsage())
  NamedRerouteDeclaration(None=LinearInterpolate(A=NamedRerouteUsage(), B=LinearInterpolate(A=@n26, B=@n24, Alpha=@n25), Alpha=ScalarParameter() {group=Image, parameterName=HasMainMask})) {name=Border_Mask, nodeColor=<rgba 0 0.2379 1 1>}
  n27 = MaterialFunctionCall() {materialFunction=&GetUserInterfaceUV}
  n28 = Add(A=@n27.Normalized UV, B=ScalarParameter() {group=Offset, parameterName=TopOffset})
  n29 = ScalarParameter() {group=Offset, parameterName=T-Angle}
  n30 = ScalarParameter() {defaultValue=256, group=Offset, parameterName=Sharpness}
  n31 = Multiply(A=ComponentMask(None=Rotator(Coordinate=@n28, Time=@n29) {speed=6.2832}) {g=true}, B=@n30)
  n32 = Rotator(
   Coordinate=Add(A=@n27.Normalized UV, B=ScalarParameter() {group=Offset, parameterName=BottomOffset})
   Time=ScalarParameter() {group=Offset, parameterName=B-Angle}
  ) {speed=6.2832}
  n33 = Multiply(A=OneMinus(None=ComponentMask(None=@n32) {g=true}), B=@n30)
  n34 = Saturate(None=Multiply(A=Saturate(None=@n31), B=Saturate(None=@n33)))
  n35 = Add(A=@n27.9-Slice UV, B=ScalarParameter() {defaultValue=-0.104, group=Offset, parameterName=LeftOffset})
  n36 = ScalarParameter() {defaultValue=-0.028, group=Offset, parameterName=L-Angle}
  n37 = Multiply(A=ComponentMask(None=Rotator(Coordinate=@n35, Time=@n36) {speed=6.2832}) {r=true}, B=@n30)
  n38 = Rotator(
   Coordinate=Add(A=@n27.9-Slice UV, B=ScalarParameter() {defaultValue=0.104, group=Offset, parameterName=RightOffset})
   Time=ScalarParameter() {defaultValue=-0.028, group=Offset, parameterName=R-Angle}
  ) {speed=6.2832}
  n39 = Multiply(A=OneMinus(None=ComponentMask(None=@n38) {r=true}), B=@n30)
  n40 = Saturate(None=Multiply(A=Saturate(None=@n37), B=Saturate(None=@n39)))
  NamedRerouteDeclaration(None=Saturate(None=Multiply(A=@n34, B=@n40))) {name=OffSets, nodeColor=<rgba 0.0365 0.0155 0 1>}
  n41 = VectorParameter() {group=Gradient, parameterName=GradientExterior, defaultValue=<rgba 1 0 0.1621 1>}
  n42 = VectorParameter() {group=Gradient, parameterName=GradientInterior, defaultValue=<rgba 1 0.9856 0 1>}
  n43 = MaterialFunctionCall() {materialFunction=&GetUserInterfaceUV}
  n44 = MaterialFunctionCall(
   UVs=@n43.Normalized UV
   CenterPosition=ScalarParameter() {defaultValue=0.5, group=Gradient, parameterName=CenterPositionGradient}
   Radius=ScalarParameter() {defaultValue=0.6, group=Gradient, parameterName=RoundedGradientRadius}
   Density=ScalarParameter() {defaultValue=1.2, group=Gradient, parameterName=RoundedGradientDensity}
  ) {materialFunction=&RadialGradientExponential}
  n45 = VectorParameter() {group=Gradient, parameterName=LinearGradient_Left, defaultValue=<rgba 0 0.4219 0.0328 0.5>}
  n46 = VectorParameter() {group=Gradient, parameterName=LinearGradient_Right, defaultValue=<rgba 0 0.009 1 1>}
  n47 = AppendVector(A=ScalarParameter() {group=Gradient, parameterName=LinearGradient_Y}, B=ScalarParameter() {group=Gradient, parameterName=LinearGradient_X})
  n48 = Rotator(
   Coordinate=Divide(A=Add(A=@n43.Normalized UV, B=@n47), B=ScalarParameter() {defaultValue=1, group=Gradient, parameterName=LinearGradientScale})
   Time=ScalarParameter() {group=Gradient, parameterName=LinearGradientRotation}
  ) {speed=6.2832}
  n49 = Clamp(None=ComponentMask(None=MaterialFunctionCall(UVs=@n48) {materialFunction=&ScaleUVsByCenter}.UVs) {r=true})
  NamedRerouteDeclaration(None=LinearInterpolate(
    A=LinearInterpolate(A=@n41.RGB, B=@n42.RGB, Alpha=@n44.RadialGradientExponential)
    B=LinearInterpolate(A=@n45.RGB, B=@n46.RGB, Alpha=@n49)
    Alpha=ScalarParameter() {desc=LinearGradient, group=Gradient, parameterName=LinearGradient, sortPriority=0}
   )) {name=Gradients_Bgs, nodeColor=<rgba 0 1 0.2231 1>}
  n50 = Rotator(Time=LinearInterpolate(
    A=ScalarParameter() {defaultValue=1.5, group=BottomBar, parameterName=Bot-Angle}
    B=ScalarParameter() {defaultValue=1.6, group=BottomBar, parameterName=Bot-Angle-hover}
    Alpha=ScalarParameter() {group=Animation, parameterName=hover, sliderMax=1}
   )) {speed=6.2832}
  n51 = ScalarParameter() {defaultValue=0.08, group=BottomBar, parameterName=Bot-RangeOffset, sliderMax=1}
  n52 = ScalarParameter() {defaultValue=0.085, group=BottomBar, parameterName=Bot-RangeOffset-hover, sliderMax=1}
  n53 = ScalarParameter() {group=Animation, parameterName=hover, sliderMax=1}
  n54 = Multiply(
   A=Subtract(A=ComponentMask(None=@n50) {r=true}, B=LinearInterpolate(A=@n51, B=@n52, Alpha=@n53))
   B=ScalarParameter() {defaultValue=128, group=BottomBar, parameterName=strenght}
  )
  NamedRerouteDeclaration(None=Saturate(None=OneMinus(None=@n54))) {name=BottomBar, nodeColor=<rgba 0.5156 0 0.0409 1>}
  n55 = VectorParameter() {group=Icon, parameterName=IconColor, defaultValue=<rgba 0 0.8445 1 1>}
  n56 = Constant() {r=1}
  n57 = Divide(A=@n2.R, B=@n2.R)
  n58 = OneMinus(None=@n57)
  n59 = Add(A=Multiply(A=@n1.Normalized UV, B=AppendVector(A=@n57, B=@n56)), B=AppendVector(A=Multiply(A=@n58) {constB=0}, B=Constant()))
  n60 = AppendVector(A=ScalarParameter() {group=Icon, parameterName=IconU-offset}, B=ScalarParameter() {group=Icon, parameterName=IconV-offset})
  n61 = TextureSampleParameter2D(UVs=MaterialFunctionCall(UVs=Add(A=@n59, B=@n60), Texture Scale=ScalarParameter() {defaultValue=1, group=Icon, parameterName=IconScale}) {materialFunction=&ScaleUVsByCenter}.UVs) {
   automaticViewMipBias=false
   group=Icon
   parameterName=ItemIcon
   samplerSource=SSM_Clamp_WorldGroupSettings
   texture=/Game/Athena/HUD/Art/BacchusHudIcons/icon_consume
  }
  n62 = OneMinus(None=@n61.A)
  n63 = NamedRerouteUsage()
  n64 = VectorParameter() {group=BottomBar, parameterName=BottomBarColor, defaultValue=<rgba 1 1 1 1>}
  n65 = NamedRerouteUsage()
  n66 = NamedRerouteUsage()
  n67 = NamedRerouteUsage()
  n68 = MaterialFunctionCall(UVs=Rotator(Coordinate=TextureCoordinate(), Time=ScalarParameter() {defaultValue=-0.5, group=Image, parameterName=V-maskRotator}) {speed=6.2832}) {materialFunction=&ScaleUVsByCenter}
  n69 = Saturate(None=Multiply(A=ComponentMask(None=@n68.UVs) {g=true}, B=ScalarParameter() {defaultValue=1, group=Image, parameterName=V_Gradient_Fade_Falloff_Rate}))
  n70 = ScalarParameter() {defaultValue=12, group=Image, parameterName=Vmask_DecayRate}
  n71 = ScalarParameter() {defaultValue=1, group=Image, parameterName=ImageScaleStart}
  n72 = ScalarParameter() {defaultValue=1.2, group=Image, parameterName=ImageScaleStop}
  n73 = ScalarParameter() {group=Animation, parameterName=AnimateImageScale}
  n74 = MaterialFunctionCall(Float2=MaterialFunctionCall(
    UVs=MaterialFunctionCall() {materialFunction=&MF_UI_UVScaling}.Scaled To Fill
    Texture Scale=LinearInterpolate(A=@n71, B=@n72, Alpha=@n73)
   ) {materialFunction=&ScaleUVsByCenter}.UVs) {materialFunction=&BreakOutFloat2Components}
  n75 = LinearInterpolate(A=@n74.R, B=OneMinus(None=@n74.R), Alpha=ScalarParameter() {group=Image, parameterName=FlipImage})
  n76 = ScalarParameter() {group=Image, parameterName=ImageU-position}
  n77 = ScalarParameter() {group=Image, parameterName=ImageV-position}
  n78 = TextureSampleParameter2D(UVs=Add(A=AppendVector(A=@n75, B=@n74.G), B=AppendVector(A=@n76, B=@n77))) {group=Image, parameterName=ItemImage, texture=&/T_Bacchus_Autofire}
  n79 = OneMinus(None=ScalarParameter() {defaultValue=1, group=Image, parameterName=ImageageOpacity})
  n80 = Saturate(None=LinearInterpolate(A=Saturate(None=@n79), Alpha=ScalarParameter() {group=Animation, parameterName=AnimateImageOpacity}) {constB=0})
  n81 = Multiply(A=Saturate(None=Subtract(A=@n78.A, B=@n80)), B=OneMinus(None=Power(Base=@n69, Exp=@n70)))
  n82 = LinearInterpolate(A=@n78.RGB, B=VectorParameter() {group=Image, parameterName=ImageColor, defaultValue=<rgba 1 1 1 1>}.RGB, Alpha=@n78.A)
  n83 = ScalarParameter() {group=Image, parameterName=HasImageColor}
  n84 = LinearInterpolate(
   A=LinearInterpolate(A=NamedRerouteUsage(), B=LinearInterpolate(A=@n78.RGB, B=@n82, Alpha=@n83), Alpha=DotProduct(A=@n81, B=@n67))
   B=LinearInterpolate(
    A=VectorParameter() {group=Border, parameterName=BorderColor, defaultValue=<rgba 0 0.3744 0.6094 1>}.RGB
    B=VectorParameter() {group=Border, parameterName=BorderColorHover, defaultValue=<rgba 1 1 1 1>}.RGB
    Alpha=ScalarParameter() {group=Animation, parameterName=hover, sliderMax=1}
   )
   Alpha=DotProduct(A=NamedRerouteUsage(), B=OneMinus(None=@n66))
  )
  n85 = StaticSwitchParameter(True=LinearInterpolate(A=LinearInterpolate(A=@n63, B=@n64.RGB, Alpha=@n63), B=@n84, Alpha=OneMinus(None=@n65)), False=@n84) {group=BottomBar, parameterName=BottomBar}
  StaticSwitchParameter(True=LinearInterpolate(A=LinearInterpolate(A=@n55.RGB, B=@n61.RGB, Alpha=@n62), B=@n85, Alpha=OneMinus(None=@n61.A)), False=@n85) {group=Icon, parameterName=HasIcon}
 ]
}
